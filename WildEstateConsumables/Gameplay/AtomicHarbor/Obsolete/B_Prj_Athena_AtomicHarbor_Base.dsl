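; Base projectile Blueprint for the AtomicHarbor consumable, filed under Obsolete, handling in-flight audio and bounce speed.
; BeginPlay starts the audio component, and a collision hit stops it before calling the parent hit handler.
; After that parent call, it clamps the projectile movement component's velocity magnitude so a bounce cannot send the projectile flying far away.
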(event EventBeginPlay
  (Audio|Components|Audio|Play (Variables|Default|GetAudio) 0.0))

(event Collision|EventHit
  (params (MyComp :type Primitive%20Component%20Object%20Reference) (Other :type Actor%20Object%20Reference) (OtherComp :type Primitive%20Component%20Object%20Reference) (bSelfMoved :type Boolean) (HitLocation :type Vector :default "0, 0, 0") (HitNormal :type Vector :default "0, 0, 0") (NormalImpulse :type Vector :default "0, 0, 0") (Hit :type Hit%20Result%20Structure%20%28by%20ref%29))
  (Audio|Components|Audio|Stop (Variables|Default|GetAudio))
  (|CallParentFunction MyComp Other OtherComp bSelfMoved HitLocation HitNormal NormalImpulse Hit)
  ;[0.1771 0.6602 1] Clamp the velocity after the bounce to avoid flying 40m away
  (Projectile|SetVelocity (Variables|Movement|GetProjectileMovementComponent) (Math|Vector|ClampVectorSize (Variables|/Script/Engine.MovementComponent|GetVelocity (Variables|Movement|GetProjectileMovementComponent)) 0.0 1e+03)))

(fn ConstructionScript
  (|CallParentFunction))
